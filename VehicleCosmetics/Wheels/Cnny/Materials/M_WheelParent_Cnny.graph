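# Parent material for the Cnny cosmetic vehicle wheel, adding animated emissive to the rim and tire over MF_WheelParent_Simple before the vehicle damage and de-rez transition functions.
# A panning water normal sampled with zero-tiled UVs gives uniform, time-varying values that pulse the glow and wobble a tight radial hotspot on the rim.
# Rim glow uses diffuse and inverted mask, tire glow uses diffuse and tread mask, blended through a named reroute, apparently vertex-color red.
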Material /VehicleCosmetics/Wheels/Cnny/Materials/M_WheelParent_Cnny
graph {
 properties {
  bAutomaticallySetUsageInEditor=false
  bUseMaterialAttributes=true
  bUsedWithMeshDeformer=true
  bUsedWithSkeletalMesh=true
  pixelDepthOffsetMode=PDOM_Legacy
  usedShadingModels=Default Lit
 }
 nodes [
  n1 = MaterialFunctionCall() {materialFunction=&MF_WheelParent_Simple}
  n2 = NamedRerouteUsage()
  MaterialFunctionCall(In=MaterialFunctionCall(Material=SetMaterialAttributes(MaterialAttributes=@n1.MaterialAttributes, Emissive Color=@n2) {attributeSetTypes [#1]}) {materialFunction=&MF_BaseVehicle_Damage}.Result) {materialFunction=&MF_VehicleDeRez_Transition}
  NamedRerouteDeclaration(None=VertexColor().R) {name=VC_R, nodeColor=<rgba 1 0 0.375 1>}
  n3 = TextureSample() {texture=/VehicleCosmetics/Wheels/Cnny/Textures/rim_cnny_m}
  n4 = TextureSample(UVs=Panner(Coordinate=TextureCoordinate() {uTiling=0, vTiling=0}) {speedX=0.2, speedY=0.5}) {samplerType=SAMPLERTYPE_Normal, texture=/VehicleCosmetics/Wheels/_SharedTextures/FX/Water_02_N}
  n5 = Add(
   A=Multiply(A=ComponentMask(None=@n4.RGB) {g=true, r=true}) {constB=0.2}
   B=Add(A=TextureCoordinate(), B=Constant2Vector() {g=-0.365, r=-0.38})
  )
  n6 = OneMinus(None=DotProduct(A=@n5, B=@n5))
  n7 = ConstantBiasScale(None=@n4.G) {bias=0.5, scale=0.75}
  n8 = Multiply(
   A=Multiply(A=TextureSample() {texture=/VehicleCosmetics/Wheels/Cnny/Textures/rim_cnny_d}.RGB, B=OneMinus(None=@n3.RGB))
   B=Multiply(A=Power(Base=@n6) {constExponent=32}, B=@n7)
  )
  n9 = Multiply(
   A=Multiply(A=@n7, B=TextureSample() {texture=/VehicleCosmetics/Wheels/Cnny/Textures/tread_mask}.RGB)
   B=TextureSample() {texture=/VehicleCosmetics/Wheels/Cnny/Textures/tire_cnny_d}.RGB
  )
  NamedRerouteDeclaration(None=LinearInterpolate(A=Multiply(A=@n8) {constB=5}, B=Multiply(A=@n9) {constB=2}, Alpha=NamedRerouteUsage())) {name=Emissive, nodeColor=<rgba 0.1953 0 1 1>}
 ]
}
